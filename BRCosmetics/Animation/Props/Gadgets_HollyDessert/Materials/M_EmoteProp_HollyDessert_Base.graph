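# Base material for the HollyDessert emote prop, using the 2023 anime character shader functions with cloth shading to animate a burn and explosion effect.
# A Timeline value, from BurnFX or a time-based debug loop (on by default via DebugAnim), is offset by TechArt mask channels to drive transition masks.
# These masks feed explosion color, damaged lit and shaded textures, and an alpha output.
# Panned, UV-distorted noise is sampled alongside the emissive blend.
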Material /BRCosmetics/Animation/Props/Gadgets_HollyDessert/Materials/M_EmoteProp_HollyDessert_Base
graph {
 properties {
  bAutomaticallySetUsageInEditor=false
  bUseMaterialAttributes=true
  bUsedWithMeshDeformer=true
  bUsedWithMorphTargets=true
  bUsedWithSkeletalMesh=true
  bUsedWithStaticLighting=true
  pixelDepthOffsetMode=PDOM_Legacy
  shadingModel=MSM_Cloth
  usedShadingModels=Cloth
 }
 nodes [
  n1 = MaterialFunctionCall!(CustomUV) {materialFunction=/Game/Global/Materials/Character_Parent_2023/Anime_Parent_2023/MaterialFunctions/MFC_AnimeShader_2023_Pre}
  n2 = MaterialFunctionCall!(CustomLightVector, OverrideNormal_WS, STT=@n1.STT, Normal_Map=@n1.STT, DFL=@n1.STT, FX=@n1.STT) {materialFunction=/Game/Global/Materials/Character_Parent_2023/Anime_Parent_2023/MaterialFunctions/MFC_AnimeShader_2023_Calculations}
  n3 = MaterialFunctionCall!(
   Shading_Gradient=@n2.Shading_Gradient
   Threshold_Gradient=@n2.Shading_Gradient
   Specular_Gradient=@n2.Shading_Gradient
   Sheen_Mask=@n2.Shading_Gradient
   Thickness=@n2.Shading_Gradient
   DFL_Mask=@n2.Shading_Gradient
   FX=@n2.Shading_Gradient
   Metallic_Gradient=@n2.Shading_Gradient
   RimLight_Gradient=@n2.Shading_Gradient
   NormalBuffer=@n2.Shading_Gradient
   --------
   CINE_LightSources=@n2.Shading_Gradient
   CINE_LightColor=@n2.Shading_Gradient
   ---------
   Color_Lit=NamedRerouteUsage()
   Color_Shaded=NamedRerouteUsage()
   SRM=@n1.SRM
   Emissive=NamedRerouteUsage()
   DFL_Color=@n1.SRM
   Specular_Color=@n1.SRM
   RimLight_Color=@n1.SRM
  ) {materialFunction=/Game/Global/Materials/Character_Parent_2023/Anime_Parent_2023/MaterialFunctions/MFC_AnimeShader_2023_Post}
  BlendMaterialAttributes(
   A=@n3.Result
   B=SetMaterialAttributes!(MaterialAttributes=@n3.Result, Opacity Mask=NamedRerouteUsage()) {attributeSetTypes [#1]}
   Alpha=OneMinus(None=NamedRerouteUsage())
  )
  n4 = NamedRerouteUsage()
  Max(A=ComponentMask(None=@n4) {r=true}, B=ComponentMask(None=@n4) {b=true})
  n5 = TextureSample() {samplerType=SAMPLERTYPE_Masks, texture=/BRCosmetics/Animation/Props/Gadgets_HollyDessert/Textures/T_HollyDessert_Emote_Prop_FX}
  NamedRerouteDeclaration(None=@n5.RGB) {name=IN_FX, nodeColor=<rgba 0 0.3672 1 1>}
  NamedRerouteDeclaration(None=@n1.Color_Lit) {name=IN_Color_Lit, nodeColor=<rgba 0.6875 1 0 1>}
  NamedRerouteDeclaration(None=Max(A=@n5.R, B=@n5.R)) {name=IN_Alpha, nodeColor=<rgba 0 0.625 1 1>}
  n6 = TextureSample() {samplerType=SAMPLERTYPE_Masks, texture=/BRCosmetics/Animation/Props/Gadgets_HollyDessert/Textures/T_HollyDessert_Prop_TechArt}
  n7 = ScalarParameter() {defaultValue=3, group=HollyDessert, parameterName=DebugLoopDuration}
  n8 = Divide(A=Fmod(A=Time(), B=@n7), B=@n7)
  n9 = NamedRerouteDeclaration(None=LinearInterpolate(
    A=ScalarParameter() {group=HollyDessert, parameterName=BurnFX}
    B=Saturate(None=@n8)
    Alpha=ScalarParameter() {defaultValue=1, group=HollyDessert, parameterName=DebugAnim}
   )) {name=Timeline, nodeColor=<rgba 0 0.9766 1 1>}
  n10 = Saturate(None=Subtract(A=Multiply(A=@n9) {constB=2}, B=@n6.G))
  n11 = Step(Y=OneMinus(None=@n6.R), X=Add(A=NamedRerouteDeclaration(None=@n10) {name=OffsetTimeline, nodeColor=<rgba 1 0 0.75 1>}) {constB=0})
  NamedRerouteDeclaration(None=Max(A=NamedRerouteUsage(), B=OneMinus(None=@n11))) {name=OUT_Alpha, nodeColor=<rgba 0 1 0.9297 1>}
  NamedRerouteDeclaration(None=Saturate(None=@n11)) {name=ColorTransition, nodeColor=<rgba 1 0 0.8437 1>}
  n12 = NamedRerouteUsage()
  n13 = OneMinus(None=Saturate(None=NamedRerouteUsage()))
  n14 = Subtract(A=@n6.B, B=Saturate(None=Power(Base=@n13) {constExponent=5}))
  n15 = LinearInterpolate(A=Constant3Vector() {constant=<rgba 0.063 0.0075 0.0033 1>}, B=Constant3Vector() {constant=<rgba 0.9911 0.1586 0 1>}, Alpha=@n14)
  n16 = ScalarParameter() {defaultValue=5, group=HollyDessert, parameterName=ExplosionIntensity}
  NamedRerouteDeclaration(None=Multiply(A=OneMinus(None=@n12), B=Multiply(A=@n15, B=@n16))) {name=Explosion_Color, nodeColor=<rgba 0.0078 0 1 1>}
  n17 = OneMinus(None=NamedRerouteUsage())
  NamedRerouteDeclaration(None=Saturate(None=Multiply(A=@n17, B=@n14))) {name=Explosion_Mask, nodeColor=<rgba 0.3125 1 0 1>}
  NamedRerouteDeclaration(None=@n5.G) {name=BurnedArea, nodeColor=<rgba 0 1 0.5547 1>}
  NamedRerouteDeclaration(None=@n1.Color_Shaded) {name=IN_Color_Shaded, nodeColor=<rgba 0.9922 0 1 1>}
  n18 = NamedRerouteUsage()
  n19 = NamedRerouteUsage()
  n20 = NamedRerouteUsage()
  NamedRerouteDeclaration(None=LinearInterpolate(
    A=LinearInterpolate(A=@n18, B=@n19, Alpha=@n20)
    B=TextureSample() {texture=/BRCosmetics/Animation/Props/Gadgets_HollyDessert/Textures/T_HollyDessert_Damaged_Emote_Prop_CL}.RGB
    Alpha=NamedRerouteUsage()
   )) {name=OUT_Color_Lit, nodeColor=<rgba 0 1 0.6719 1>}
  n21 = NamedRerouteUsage()
  n22 = NamedRerouteUsage()
  n23 = NamedRerouteUsage()
  n24 = TextureSample() {texture=/BRCosmetics/Animation/Props/Gadgets_HollyDessert/Textures/T_HollyDessert_Damaged_Emote_Prop_CS}
  NamedRerouteDeclaration(None=LinearInterpolate(A=LinearInterpolate(A=@n21, B=@n22, Alpha=@n23), B=@n24.RGB, Alpha=NamedRerouteUsage())) {name=OUT_Color_Shaded, nodeColor=<rgba 1 0.3516 0 1>}
  NamedRerouteDeclaration(None=@n1.Emissive) {name=IN_Emissive, nodeColor=<rgba 0 1 0.0859 1>}
  n25 = TextureCoordinate()
  n26 = VectorParameter() {group=HollyDessert, parameterName=DistortionTiling, channelNames {a="", b="", g=Tile Y, r=Tile X}, defaultValue=<rgba 4 3 2 1>}
  n27 = ComponentMask(None=@n26) {g=true, r=true}
  n28 = VectorParameter() {group=HollyDessert, parameterName=DistortionSpeed, channelNames {a="", b="", g=Speed Y, r=Speed X}, defaultValue=<rgba -0.1 0.1 1 1>}
  n29 = TextureSampleParameter2D(UVs=Panner(Coordinate=Multiply(A=@n25, B=@n27), Speed=ComponentMask(None=@n28) {g=true, r=true})) {
   group=HollyDessert
   parameterName=DistortionTexture
   samplerType=SAMPLERTYPE_LinearColor
   texture=&T_NoisePack_LinearColor
  }
  NamedRerouteDeclaration(None=ChannelMaskParameter(None=@n29.RGB) {group=HollyDessert, parameterName=DistortionMaskChannel}.R) {name=DistortionMask, nodeColor=<rgba 0 1 0.0859 1>}
  NamedRerouteDeclaration(None=ComponentMask(None=ConstantBiasScale(None=@n29.RGB) {bias=-0.5, scale=2}) {g=true, r=true}) {name=UVDistortion, nodeColor=<rgba 1 0 0.75 1>}
  n30 = VectorParameter() {group=HollyDessert, parameterName=NoiseTiling, channelNames {a="", b="", g=Tile Y, r=Tile X}, defaultValue=<rgba 4 4 4 1>}
  n31 = Multiply(A=TextureCoordinate(), B=ComponentMask(None=@n30) {g=true, r=true})
  n32 = Multiply(A=ScalarParameter() {defaultValue=0.1, group=HollyDessert, parameterName=DistortionStrength}, B=NamedRerouteUsage())
  n33 = Time()
  n34 = VectorParameter() {group=HollyDessert, parameterName=NoiseSpeed, channelNames {a="", b="", g=Speed Y, r=Speed X}, defaultValue=<rgba 0.025 -0.05 1 1>}
  n35 = ComponentMask(None=@n34) {g=true, r=true}
  n36 = TextureSampleParameter2D(UVs=Add(A=Add(A=@n31, B=@n32), B=Multiply(A=@n33, B=@n35))) {
   group=HollyDessert
   parameterName=NoiseTexture
   samplerType=SAMPLERTYPE_LinearGrayscale
   texture=/Game/Characters/Player/Male/Medium/Bodies/M_MED_Hurricane/Textures/T_MED_Hurricane_StormCloud
  }
  n37 = ScalarParameter() {defaultValue=1, group=HollyDessert, parameterName=NoiseRangeScale}
  NamedRerouteDeclaration(None=Add(A=Multiply(A=@n36.R, B=@n37), B=ScalarParameter() {group=HollyDessert, parameterName=NoiseRangeBias})) {name=NoiseSample, nodeColor=<rgba 0 0.3438 1 1>}
  n38 = NamedRerouteUsage()
  n39 = Multiply(A=@n24.RGB, B=NamedRerouteUsage())
  n40 = ScalarParameter() {defaultValue=1, group=HollyDessert, parameterName=EmissiveIntensity}
  n41 = NamedRerouteUsage()
  n42 = Add(A=NamedRerouteUsage(), B=Multiply(A=Multiply(A=@n39, B=@n40), B=ComponentMask(None=@n41) {r=true}))
  n43 = NamedRerouteUsage()
  n44 = NamedRerouteUsage()
  n45 = NamedRerouteUsage()
  NamedRerouteDeclaration(None=LinearInterpolate(A=LinearInterpolate(A=@n38, B=@n42, Alpha=@n43), B=Multiply(A=@n44, B=@n45), Alpha=NamedRerouteUsage())) {name=OUT_Emissive, nodeColor=<rgba 0.4062 1 0 1>}
 ]
}
